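# UI material that draws the MistRaven GrimStrut reticle bracket and changes it with the WraithFormCooldown parameter.
# Mirrored 9-slice UVs sample two MSDF textures, and their distance fields are blended by the cooldown value, morphing the normal bracket into the cooldown bracket.
# Two anti-aliased thresholds on that field give the bracket and its shadow.
# Cooldown also tints the bracket toward CooldownColor and halves opacity, alongside shadow and master opacity controls.
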Material /MistRaven/HUD/GrimStrut/Materials/M_UI_MistRaven_GrimStrut_Reticle_Bar
graph {
 properties {bAutomaticallySetUsageInEditor=false, blendMode=BLEND_Translucent, materialDomain=MD_UI, usedShadingModels=Default Lit}
 nodes [
  n1 = MaterialFunctionCall!(
   UV=MaterialFunctionCall!(UVs) {materialFunction=/FortUILibrary/MaterialFunctions/Basics/MF_UI_9SliceUVs}.UVs
   Offset
   Rotation
   Scale=Multiply(A=Constant2Vector() {g=1, r=2}) {constB=1.15}
   MirrorY
  ) {materialFunction=/FortUILibrary/MaterialFunctions/Basics/MF_UI_Symmetry_2x1_DEPRECATED}
  n2 = ScalarParameter() {parameterName=WraithFormCooldown, sliderMax=1}
  n3 = LinearInterpolate(
   A=MaterialFunctionCall!(MSDF RGB=TextureSample(UVs=@n1.Result) {texture=/MistRaven/HUD/GrimStrut/Textures/T_UI_MSDF_MistRaven_GrimStrut_Bracket_New}.RGB) {materialFunction=/FortUILibrary/MaterialFunctions/Shapes/Modifiers/MF_UI_MSDFtoSDF}.SDF
   B=MaterialFunctionCall(MSDF RGB=TextureSample(UVs=@n1.Result) {texture=/MistRaven/HUD/GrimStrut/Textures/T_UI_MSDF_MistRaven_GrimStrut_CooldownBracket}.RGB) {materialFunction=&MF_UI_MSDFtoSDF}.SDF
   Alpha=@n2
  ) {constAlpha=1}
  Composite!(SDF=@n3) {subgraphName=Bracket AA + Shadow}
  Constant() {r=0.2}
  Constant() {r=0.5}
  MaterialExpressionReroute_4 = Reroute(None=@n3)
  n4 = Constant() {r=0.25}
  n5 = MaterialFunctionCall(DistanceField=@MaterialExpressionReroute_4.SDF, Threshold=@n4) {materialFunction=&MF_UI_AntiAliasing}
  MaterialExpressionReroute_5 = Reroute(None=@n5.Outer)
  n6 = MaterialFunctionCall(DistanceField=@MaterialExpressionReroute_4.SDF, Threshold=Add(A=@n4) {constB=-0.1}) {materialFunction=&MF_UI_AntiAliasing}
  MaterialExpressionReroute_6 = Reroute(None=@n6.Outer)
  n7 = ScalarParameter() {defaultValue=0.5, parameterName=ShadowOpacity}
  Multiply(
   A=Multiply(
    A=LinearInterpolate(A=@MaterialExpressionReroute_5.AA'd Art, B=@MaterialExpressionReroute_6.Shadow, Alpha=@n7)
    B=LinearInterpolate(Alpha=@n2) {constA=1, constB=0.5}
   )
   B=ScalarParameter() {defaultValue=0.85, parameterName=MasterOpacity}
  )
  NamedRerouteDeclaration(None=@MaterialExpressionReroute_5.AA'd Art) {name=BracketColor, nodeColor=<rgba 0 1 0.4844 1>}
  PinBase() {reroutePins [SDF {expression=&self:MaterialExpressionReroute_4}]}
  PinBase!(AA'd Art=@n5.Outer, Shadow=@n6.Outer) {reroutePins [
    AA'd Art {expression=&self:MaterialExpressionReroute_5}
    Shadow {expression=&self:MaterialExpressionReroute_6}
   ]}
  n8 = VectorParameter() {parameterName=CooldownColor, defaultValue=<rgba 0.5 0.5 0.5 0>}
  n9 = ScalarParameter() {parameterName=WraithFormCooldown, sliderMax=1}
  Reroute(None=Multiply(A=NamedRerouteUsage(), B=LinearInterpolate(B=@n8.RGB, Alpha=@n9) {constA=1, constAlpha=1}))
 ]
}
