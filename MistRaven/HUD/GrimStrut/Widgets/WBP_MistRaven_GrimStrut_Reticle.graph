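# Weapon reticle extension widget for MistRaven GrimStrut that shows Wraith Form state, remaining duration and cooldown through animated bracket and wing materials.
# Async gameplay-tag waits on the player controller drive delta-gated handlers that play start, end and cooldown animations, deferring cooldown visuals while the form is active.
# A looping timer polls the active gameplay effect to set the wings' Fill.
# On construct, an ascended-equipped pawn tag selects the fill color.
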WidgetBlueprint /MistRaven/HUD/GrimStrut/Widgets/WBP_MistRaven_GrimStrut_Reticle
animations [
 Anim_Equip {
  binding_count=2
  length_seconds=0.25
  object=&self:Anim_Equip
  sequence {bindings [
    Image_Wings {
     id=#1
     possessed_class=Image
     tracks [{
       type=MovieSceneWidgetMaterialTrack
       sections [{
         type=MovieSceneParameterSection
         end_frame=5
         start_frame=0
         channels [MasterOpacity {type=MovieSceneScriptingFloatChannel, keys [[0,0],[4,0],[5,0.85]]}]
        }]
      }]
    }
    Image_Brackets {
     id=#2
     possessed_class=FortLazyImage
     tracks [{
       type=MovieSceneWidgetMaterialTrack
       sections [{
         type=MovieSceneParameterSection
         end_frame=5
         start_frame=0
         channels [
          MasterOpacity {type=MovieSceneScriptingFloatChannel, keys [[0,0],[5,1]]}
          SpreadOffset {type=MovieSceneScriptingFloatChannel, keys [[0,-0.15],[3,0.6],[5,0.5]]}
          WraithFormCooldown {type=MovieSceneScriptingFloatChannel, keys [[0,1],[5,0]]}
         ]
        }]
      }]
    }
   ]}
 }
 Anim_WraithFormStarted {
  binding_count=1
  length_seconds=0.25
  object=&self:Anim_WraithFormStarted
  sequence {bindings [Image_Wings {
     id=#3
     possessed_class=Image
     tracks [
      {
       type=MovieSceneColorTrack
       display_name=Color and Opacity
       sections [{
         type=MovieSceneColorSection
         end_frame=5
         start_frame=0
         channels [
          Color.R {type=MovieSceneScriptingFloatChannel, default=1}
          Color.G {type=MovieSceneScriptingFloatChannel, default=1}
          Color.B {type=MovieSceneScriptingFloatChannel, default=1}
          Color.A {type=MovieSceneScriptingFloatChannel, keys [[0,0],[5,1]]}
         ]
        }]
      }
      {
       type=MovieSceneWidgetMaterialTrack
       sections [{
         type=MovieSceneParameterSection
         end_frame=5
         start_frame=0
         channels [
          EdgeSoftnessFalloff {type=MovieSceneScriptingFloatChannel, keys [[0,1],[5,0.1335]]}
          x {type=MovieSceneScriptingFloatChannel, keys [[0,0.2],[5,0.943]]}
         ]
        }]
      }
     ]
    }]}
 }
 Anim_WraithFormEnded {
  binding_count=1
  length_seconds=0.25
  object=&self:Anim_WraithFormEnded
  sequence {bindings [Image_Wings {
     id=#4
     possessed_class=Image
     tracks [
      {
       type=MovieSceneColorTrack
       display_name=Color and Opacity
       sections [{
         type=MovieSceneColorSection
         end_frame=5
         start_frame=0
         channels [
          Color.R {type=MovieSceneScriptingFloatChannel, default=1}
          Color.G {type=MovieSceneScriptingFloatChannel, default=1}
          Color.B {type=MovieSceneScriptingFloatChannel, default=1}
          Color.A {type=MovieSceneScriptingFloatChannel, keys [[0,1],[5,0]]}
         ]
        }]
      }
      {
       type=MovieSceneWidgetMaterialTrack
       sections [{
         type=MovieSceneParameterSection
         end_frame=5
         start_frame=0
         channels [
          EdgeSoftnessFalloff {type=MovieSceneScriptingFloatChannel, keys [[0,0.1335],[5,1]]}
          x {type=MovieSceneScriptingFloatChannel, keys [[0,0.943],[5,0.2]]}
         ]
        }]
      }
     ]
    }]}
 }
 Anim_EnterCooldown {
  binding_count=1
  length_seconds=0.3
  object=&self:Anim_EnterCooldown
  sequence {bindings [Image_Brackets {
     id=#5
     possessed_class=FortLazyImage
     tracks [{
       type=MovieSceneWidgetMaterialTrack
       sections [{
         type=MovieSceneParameterSection
         end_frame=6
         start_frame=0
         channels [WraithFormCooldown {type=MovieSceneScriptingFloatChannel, keys [[0,0],[6,1]]}]
        }]
      }]
    }]}
 }
 Anim_ExitCooldown {
  binding_count=1
  length_seconds=0.3
  object=&self:Anim_ExitCooldown
  sequence {bindings [Image_Brackets {
     id=#6
     possessed_class=FortLazyImage
     tracks [{
       type=MovieSceneWidgetMaterialTrack
       sections [{
         type=MovieSceneParameterSection
         end_frame=6
         start_frame=0
         channels [WraithFormCooldown {type=MovieSceneScriptingFloatChannel, keys [[0,1],[6,0]]}]
        }]
      }]
    }]}
 }
]
graphs [
 EventGraph {nodes [
   OnEquip() {is=title K2Node_CustomEvent}
   > SetRenderOpacity(self=Get Image_Brackets() -> (Image_Brackets:Fort Lazy Image Object Reference), InOpacity=0)
   > DelayUntilNextTick!(execute) -> (then)
   > n1 = HasMatchingGameplayTag!(
    execute:Exec
    self:Object Reference=Get Player Controller() -> (Player Controller:Player Controller Object Reference)
    TagToCheck:Gameplay Tag Structure=Get WraithFormCooldownTag() -> (WraithFormCooldownTag:Gameplay Tag Structure)
   ) -> (then:Exec, ReturnValue:bool) {is=title K2Node_Message}
   > n2 = Set IsCooldown(IsCooldown:bool=@n1) -> (Output_Get:bool)
   > SetScalarParameterValue(
    self=Get MID_Brackets() -> (MID_Brackets:Material Instance Dynamic Object Reference)
    ParameterName=WraithFormCooldown
    Value=SelectFloat(A=1, B=0, bPickA=@n2)
   )
   > OnWraithFormActive()
   > OnWraithFormCooldown()
   > PlayAnimationForward(InAnimation=Get Anim_Equip() -> (Anim_Equip:Widget Animation Object Reference), bRestoreState=true)
   > SetRenderOpacity(self=Get Image_Brackets(), InOpacity=1)
   OnWraithFormActive() {is=title K2Node_CustomEvent}
   > n3 = WaitGameplayTagAddToActor!(
    execute:Exec
    TargetActor:Actor Object Reference=Get Player Controller()
    Tag:Gameplay Tag Structure=Get WraithFormActiveTag() -> (WraithFormActiveTag:Gameplay Tag Structure)
    OnlyTriggerOnce:bool=false
   ) -> (then:Exec, AsyncTaskProxy:Ability Async Wait Gameplay Tag Added Object Reference, Added:Exec) {is=title K2Node_AsyncAction}
   > n4 = WaitGameplayTagRemoveFromActor!(
    execute:Exec
    TargetActor:Actor Object Reference=Get Player Controller()
    Tag:Gameplay Tag Structure=Get WraithFormActiveTag()
    OnlyTriggerOnce:bool=false
   ) -> (then:Exec, AsyncTaskProxy:Ability Async Wait Gameplay Tag Removed Object Reference, Removed:Exec) {is=title K2Node_AsyncAction}
   HandleWraithForm!(execute=@n4.Removed, self, IsWraithFormActive:bool=false) -> (then)
   > Clear and Invalidate Timer by Handle(Handle=Get WraithFormDurationTimerHandle() -> (WraithFormDurationTimerHandle:Timer Handle Structure))
   HandleWraithForm(execute=@n3.Added, IsWraithFormActive=true)
   > n5 = Set Timer by Event(Delegate=Create Event(), Time=0.1, bLooping=true, bMaxOncePerFrame=true)
   > Set WraithFormDurationTimerHandle(WraithFormDurationTimerHandle:Timer Handle Structure=@n5) -> (Output_Get:Timer Handle Structure)
   OnWraithFormCooldown() {is=title K2Node_CustomEvent}
   > n6 = WaitGameplayTagAddToActor(TargetActor=Get Player Controller(), Tag=Get WraithFormCooldownTag(), OnlyTriggerOnce=false)
   > n7 = WaitGameplayTagRemoveFromActor(TargetActor=Get Player Controller(), Tag=Get WraithFormCooldownTag(), OnlyTriggerOnce=false)
   HandleCooldown!(execute=@n7.Removed, self, IsCooldown:bool=false) -> (then)
   HandleCooldown(execute=@n6.Added, IsCooldown=true)
   HandleWraithFormDuration() {is=title K2Node_CustomEvent}
   > n8 = GetActiveEffectsWithAllTags!(
    execute
    self:Ability System Component Object Reference=GetAbilitySystemComponent!(Actor:Actor Object Reference=Get Player Controller()) -> (ReturnValue:Ability System Component Object Reference)
    Tags:Gameplay Tag Container Structure=MakeGameplayTagContainerFromTag!(SingleTag:Gameplay Tag Structure=Get WraithFormActiveDuration() -> (WraithFormActiveDuration:Gameplay Tag Structure)) -> (ReturnValue:Gameplay Tag Container Structure)
   ) -> (then, ReturnValue:Array of Active Gameplay Effect Handle Structures)
   > Branch(Condition="integer > integer"(A=Length(TargetArray:Array of Active Gameplay Effect Handle Structures=@n8)))
   n9 = "Get (a ref)"(Array:Array of Active Gameplay Effect Handle Structures=@n8, Dimension 1=0) -> (Output:"Active Gameplay Effect Handle Structure (by ref)")
   > n10 = GetActiveGameplayEffectRemainingDuration!(execute, ActiveHandle:Active Gameplay Effect Handle Structure=@n9) -> (then, ReturnValue:float)
   > n11 = GetActiveGameplayEffectTotalDuration!(execute, ActiveHandle:Active Gameplay Effect Handle Structure=@n9) -> (then, ReturnValue:float)
   > n12 = GetDynamicMaterial(self=Get Image_Wings() -> (Image_Wings:Image Object Reference))
   > SetScalarParameterValue(self=@n12, ParameterName=Fill, Value=float / float(A=@n10, B=@n11))
   Event Construct()
   > Parent: Construct()
   > Set Player Controller(Player Controller:Player Controller Object Reference=GetOwningPlayer()) -> (Output_Get:Player Controller Object Reference)
   > Set PlayerPawn(PlayerPawn:Pawn Object Reference=GetOwningPlayerPawn!(self:User Widget Object Reference) -> (ReturnValue:Pawn Object Reference)) -> (Output_Get:Pawn Object Reference)
   > n13 = GetDynamicMaterial(self=Get Image_Brackets())
   > Set MID_Brackets(MID_Brackets:Material Instance Dynamic Object Reference=@n13) -> (Output_Get:Material Instance Dynamic Object Reference)
   > n14 = GetDynamicMaterial(self=Get Image_Wings())
   > Set MID_Wings(MID_Wings:Material Instance Dynamic Object Reference=@n14) -> (Output_Get:Material Instance Dynamic Object Reference)
   > OnEquip() {is=title K2Node_CallFunction}
   > n15 = Cast To FortPlayerPawn(Object=Get PlayerPawn() -> (PlayerPawn:Pawn Object Reference))
   > HandleAscended!(
    execute
    self
    IsAscended:bool=HasMatchingGameplayTag!(
     self:Gameplay Tag Asset Interface Interface=@n15
     TagToCheck:Gameplay Tag Structure=Get AscendedEquippedTag() -> (AscendedEquippedTag:Gameplay Tag Structure)
    ) -> (ReturnValue:bool) {is=title K2Node_CallFunction}
   ) -> (then)
  ]}
 HandleAscended {nodes [
   n1 = HandleAscended!() -> (then, IsAscended:bool=~) {is=title K2Node_FunctionEntry}
   > SetVectorParameterValue(
    self=Get MID_Wings() -> (MID_Wings:Material Instance Dynamic Object Reference)
    ParameterName=FillColor
    Value=SelectColor(
     A=Get AscendedFillcolor() -> (AscendedFillcolor:Linear Color Structure)
     B=Get NormalFillColor() -> (NormalFillColor:Linear Color Structure)
     bPickA=@n1
    )
   )
  ]}
 HandleCooldown {nodes [
   n1 = HandleCooldown!() -> (then, IsCooldown:bool=~) {is=title K2Node_FunctionEntry}
   > n2 = Delta Gate Wildcard!(execute:Exec, Variable:"Boolean (by ref)"=Get IsCooldown() -> (IsCooldown:bool), New Value:bool=@n1) -> (Success:Exec, Updated Variable:"Boolean (by ref)", Previous Variable:bool=~) {is=title K2Node_MacroInstance}
   > n3 = Branch(Condition=@n2)
   > n4 = Branch(Condition=Get IsWraithFormActive() -> (IsWraithFormActive:bool))
   PlayAnimationForward(execute=@n4.else, InAnimation=Get Anim_EnterCooldown() -> (Anim_EnterCooldown:Widget Animation Object Reference))
   PlayAnimationForward(execute=@n3.else, InAnimation=Get Anim_ExitCooldown() -> (Anim_ExitCooldown:Widget Animation Object Reference))
  ]}
 HandleCooldownStart {nodes [
   n1 = HandleCooldownStart!() -> (then, WraithFormActive:bool=~) {is=title K2Node_FunctionEntry}
   > n2 = Branch(Condition=@n1)
   PlayAnimationForward(execute=@n2.else, InAnimation=Get Anim_EnterCooldown())
  ]}
 HandleCooldownStop {nodes [
   HandleCooldownStop() {is=title K2Node_FunctionEntry}
   > PlayAnimationForward(InAnimation=Get Anim_ExitCooldown())
  ]}
 HandleWraithForm {nodes [
   n1 = HandleWraithForm!() -> (then, IsWraithFormActive:bool=~) {is=title K2Node_FunctionEntry}
   > n2 = Delta Gate Wildcard(Variable=Get IsWraithFormActive(), New Value=@n1) -> (Previous Variable=~)
   > n3 = Branch(Condition=@n2)
   > PlayAnimationForward(InAnimation=Get Anim_WraithFormStarted() -> (Anim_WraithFormStarted:Widget Animation Object Reference))
   PlayAnimationForward(execute=@n3.else, InAnimation=Get Anim_WraithFormEnded() -> (Anim_WraithFormEnded:Widget Animation Object Reference))
   > n4 = Branch(Condition=Get IsCooldown())
   > PlayAnimationForward(InAnimation=Get Anim_EnterCooldown())
   SetColorAndOpacity(
    execute=@n4.else
    self:Image Object Reference=Get Image_Brackets()
    InColorAndOpacity:Linear Color Structure=Get ResetColor() -> (ResetColor:Linear Color Structure)
   )
  ]}
 HandleWraithFormActive {nodes [
   HandleWraithFormActive() {is=title K2Node_FunctionEntry}
   > Set IsWraithFormActive(IsWraithFormActive:bool=true) -> (Output_Get:bool)
   > PlayAnimationForward(InAnimation=Get Anim_WraithFormStarted())
  ]}
 HandleWraithFormInactive {nodes [
   HandleWraithFormInactive!() -> (then, CooldownActive:bool=~) {is=title K2Node_FunctionEntry}
   > PlayAnimationForward(InAnimation=Get Anim_WraithFormEnded())
   > n1 = Branch(Condition=Get IsCooldown())
   > PlayAnimationForward(InAnimation=Get Anim_EnterCooldown())
   SetColorAndOpacity(execute=@n1.else, self=Get Image_Brackets(), InColorAndOpacity=Get ResetColor())
  ]}
]
variables [
 AscendedEquippedTag {type=GameplayTag, default {tagName=MistRaven.AllAscendedEquipped}}
 AscendedFillcolor {type=LinearColor, default=<rgba 0.5029 0 1 1>}
 IsCooldown {type=boolean}
 IsWraithFormActive {type=boolean}
 MID_Brackets {type=MaterialInstanceDynamic}
 MID_Wings {type=MaterialInstanceDynamic}
 NormalFillColor {type=LinearColor, default=<rgba 1 0 0 1>}
 Player Controller {type=PlayerController, default=null}
 PlayerPawn {type=Pawn, default=null}
 WraithFormActiveDuration {type=GameplayTag, default {tagName=MistRaven.GrimStrut.GEActiveWindow}}
 WraithFormActiveTag {type=GameplayTag, default {tagName=MistRaven.GrimStrut.Active}}
 WraithFormCooldownTag {type=GameplayTag, default {tagName=Granted.Cooldown.MistRaven.GrimStrut.Ascended.Cooldown}}
 WraithFormCooldownTagContainer {type=GameplayTagContainer, default {gameplayTags [{tagName=MistRaven.GrimStrut.GEActiveWindow}]}}
 WraithFormDurationTimerHandle {type=TimerHandle}
]
widgets {
 info {
  inheritedWidgetCount=1
  parentClass=/Script/FortniteUI.FortWeaponReticleExtensionWidgetBase
  rootWidgetClass=&Overlay
  widgetCount=4
 }
 widgets [
  Overlay_114 {widgetClassPath=&Overlay, widget=Overlay_114}
  Image_Brackets {
   widgetClassPath=&FortLazyImage
   bIsVariable=true
   parent=Overlay_114
   widget=Image_Brackets
   properties {brush {
     bIsDynamicallyLoaded=false
     drawAs=Box
     imageType=FullColor
     mirroring=NoMirror
     resourceName=None
     resourceObject=/MistRaven/HUD/GrimStrut/Materials/M_UI_MistRaven_GrimStrut_Reticle_Bar
     tiling=NoTile
     imageSize=<210 84>
     margin=<ltrb 0.5 0 0.5 0>
     outlineSettings {
      bUseBrushTransparency=false
      roundingType=HalfHeightRadius
      width=0
      color {colorUseRule=UseColor_Specified, specifiedColor=<rgba 0 0 0 0>}
      cornerRadii=<0 0 0 0>
     }
     tintColor {colorUseRule=UseColor_Specified, specifiedColor=<rgba 1 1 1 1>}
     uVRegion {bIsValid=false, max=<0 0>, min=<0 0>}
    }}
   slot {class=OverlaySlot, properties {horizontalAlignment=HAlign_Center, verticalAlignment=VAlign_Center}}
  }
  Image_Wings {
   widgetClassPath=&Image
   bIsVariable=true
   parent=Overlay_114
   widget=Image_Wings
   properties {
    visibility=HitTestInvisible
    brush {
     bIsDynamicallyLoaded=false
     drawAs=Box
     imageType=FullColor
     mirroring=NoMirror
     resourceName=None
     resourceObject=/MistRaven/HUD/GrimStrut/Materials/MI_UI_MistRaven_GrimStrut_Reticle_Wings
     tiling=NoTile
     imageSize=<210 100>
     margin=<ltrb 0.5 0 0.5 0>
     outlineSettings {
      bUseBrushTransparency=false
      roundingType=HalfHeightRadius
      width=0
      color {colorUseRule=UseColor_Specified, specifiedColor=<rgba 0 0 0 0>}
      cornerRadii=<0 0 0 0>
     }
     tintColor {colorUseRule=UseColor_Specified, specifiedColor=<rgba 1 1 1 1>}
     uVRegion {bIsValid=false, max=<0 0>, min=<0 0>}
    }
    colorAndOpacity=<rgba 1 1 1 0>
    renderTransform {angle=0, scale=<1.5 1>, shear=<0 0>, translation=<0 -10>}
   }
   slot {class=OverlaySlot, properties {horizontalAlignment=HAlign_Center, verticalAlignment=VAlign_Center}}
  }
  TutorialHighlight {widgetClassPath=&TutorialHighlight}
 ]
}
